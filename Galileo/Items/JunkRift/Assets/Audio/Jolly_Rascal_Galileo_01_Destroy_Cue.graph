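# Sound cue for the destroy sound of the Galileo 01 Jolly Rascal prop in the JunkRift item.
# A single Wave Player plays the large-prop impact wave directly at slightly reduced volume, with no randomization or mixing.
# It uses shared Jolly Rascal attenuation and the third-person ranged-weapon fire sound class.
# A concurrency override caps simultaneous instances, taking precedence over the assigned one-per-owner stop-oldest set.
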SoundCue /Galileo/Items/JunkRift/Assets/Audio/Jolly_Rascal_Galileo_01_Destroy_Cue
root=WavePlayer_2
nodes [{id=WavePlayer_2, properties {soundWaveAssetPtr=&/WPN_Jolly_Rascal_Large_Prop_Galileo_01_Impact}}]
properties {
 attenuationSettings=/FNGameplayCues/Sounds/Attenuation/Fort_Jolly_Rascal_Att
 bOverrideConcurrency=true
 maxDistance=107500
 soundClassObject=&Fort_WeaponRanged_Fire_3P_PSM
 volumeMultiplier=0.8
 concurrencyOverrides {maxCount=3}
 concurrencySet [/Game/Sounds/Concurrency_Rules/OneMaximum_Owner_StopOldest]
}
